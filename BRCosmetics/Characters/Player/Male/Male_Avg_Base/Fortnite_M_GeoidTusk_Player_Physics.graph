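# Physics asset for the GeoidTusk male player character in Fortnite, defining simplified collision bodies for its skeleton.
# Capsules cover the pelvis, spine, arms, hands, legs and feet, with a single large sphere on the head.
# Only two constraints exist, right hand to forearm and right foot to calf, both with free swing and twist, so most bodies have no joints linking them.
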PhysicsAsset /BRCosmetics/Characters/Player/Male/Male_Avg_Base/Fortnite_M_GeoidTusk_Player_Physics
bodies {
 Spine_03 [{shapeType=Capsule, length=13.4928, radius=32.9688, center=<0.817 1.6012 -0.0107>, rotation=<rot -0.0003 88.5965 179.7999>}]
 calf_l [{shapeType=Capsule, length=5.6469, radius=13.0073, center=<-17.0652 0 0>, rotation=<rot 84.9758 -5.3301 4.6577>}]
 calf_r [{shapeType=Capsule, length=5.6469, radius=13.0073, center=<17.0652 0 0>, rotation=<rot 84.9758 -5.3301 -4.6577>}]
 foot_l [{shapeType=Capsule, length=10, radius=10, shapeName=hand_l_capsule, center=<-4.8262 4.0356 -0.2592>, rotation=<rot 0 0 90>}]
 foot_r [{shapeType=Capsule, length=10, radius=10, shapeName=hand_l_capsule, center=<4.8262 -4.0356 0.2592>, rotation=<rot 0 0 -90>}]
 hand_l [{shapeType=Capsule, length=6.0543, radius=9.76, shapeName=hand_l_capsule, center=<7.5363 -1.7418 0>, rotation=<rot 90 0 0>}]
 hand_r [{shapeType=Capsule, length=6.0543, radius=9.76, shapeName=hand_l_capsule, center=<-7.5363 1.7418 0>, rotation=<rot 90 0 180>}]
 head [{shapeType=Sphere, radius=28.5071, center=<0.4808 6.8525 0>}]
 lowerarm_l [{shapeType=Capsule, length=16.9921, radius=13.8318, center=<9.7344 0 0>, rotation=<rot -90 -0.0148 0.0149>}]
 lowerarm_r [{shapeType=Capsule, length=16.9921, radius=13.8318, center=<-9.7344 0 0>, rotation=<rot -90 -0.0148 0.0149>}]
 pelvis [{shapeType=Capsule, length=11, radius=38.7805, center=<15.0828 -0.6244 0>, rotation=<rot 0 90 0>}]
 spine_05 [{shapeType=Capsule, length=35.3077, radius=27.0907, center=<-3.2021 -1.7806 0.0001>}]
 thigh_l [{shapeType=Capsule, length=14, radius=16, center=<-19.5847 0 -0.8714>, rotation=<rot 80 -0.0001 0.0001>}]
 thigh_r [{shapeType=Capsule, length=14, radius=16, center=<19.5847 0 0.8714>, rotation=<rot 80 -0.0001 0.0001>}]
 upperarm_l [{shapeType=Capsule, length=24.3038, radius=15.1074, center=<10.1374 0 -0.2114>, rotation=<rot -88.7876 -179.9763 179.9773>}]
 upperarm_r [{shapeType=Capsule, length=24.3038, radius=15.1074, center=<-10.1374 0 0.2114>, rotation=<rot -88.7876 -179.9763 179.9773>}]
}
constraints [
 {
  bone1Name=hand_r
  bone2Name=lowerarm_r
  swing1LimitDegrees=45
  swing1Motion=Free
  swing2LimitDegrees=45
  swing2Motion=Free
  twistLimitDegrees=45
  twistMotion=Free
 }
 {
  bone1Name=foot_r
  bone2Name=calf_r
  swing1LimitDegrees=45
  swing1Motion=Free
  swing2LimitDegrees=45
  swing2Motion=Free
  twistLimitDegrees=45
  twistMotion=Free
 }
]
