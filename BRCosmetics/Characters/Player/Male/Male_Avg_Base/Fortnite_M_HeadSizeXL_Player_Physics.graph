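# Physics asset for the average-build male Fortnite player, in a variant named for an extra-large head size, giving the skeleton per-bone collision bodies.
# Capsules cover the pelvis, spine, neck, head and limb segments, while boxes cover the hands and feet.
# The head uses one large capsule named head_capsule, wider than the spine capsules.
# It defines no constraints, so the bodies have no joints linking them for a ragdoll.
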PhysicsAsset /BRCosmetics/Characters/Player/Male/Male_Avg_Base/Fortnite_M_HeadSizeXL_Player_Physics
bodies {
 Spine_01 [{shapeType=Capsule, length=4.1395, radius=11.6599, center=<2.9268 0.1619 0.1571>, rotation=<rot -0.0004 89.438 0.0008>}]
 Spine_03 [{shapeType=Capsule, length=6.5119, radius=10.831, center=<0.793 0.6215 -0.0107>, rotation=<rot -0.0003 88.5965 179.7999>}]
 calf_l [{shapeType=Capsule, length=38.6878, radius=7.1396, center=<-20.9094 -1.1698 -0.4407>, rotation=<rot -90 0 0>}]
 calf_r [{shapeType=Capsule, length=38.6878, radius=7.1396, center=<20.9094 1.1698 0.4407>, rotation=<rot -90 0 0>}]
 foot_l [{shapeType=Box, extentX=9.6864, extentY=13.4423, extentZ=27.7399, center=<-4.6927 7.8966 0.1247>, rotation=<rot 0.0499 179.0396 -91.5724>}]
 foot_r [{shapeType=Box, extentX=9.6864, extentY=13.4423, extentZ=27.7399, center=<4.6927 -7.8966 -0.1247>, rotation=<rot 0.0499 179.0396 -91.5724>}]
 hand_l [{shapeType=Box, extentX=15.9284, extentY=8.2096, extentZ=10.1104, center=<7.8245 -1.3502 0>}]
 hand_r [{shapeType=Box, extentX=15.9284, extentY=8.2096, extentZ=10.1104, center=<-7.8245 1.3502 0>}]
 head [{shapeType=Capsule, length=13, radius=12.5, shapeName=head_capsule, center=<8.9343 1.952 0>, rotation=<rot 90 3.5763 18.5778>}]
 lowerarm_l [{shapeType=Capsule, length=16.9921, radius=6.3, center=<9.7344 0 0>, rotation=<rot -90 -0.0148 0.0149>}]
 lowerarm_r [{shapeType=Capsule, length=16.9921, radius=6.3, center=<-9.7344 0 0>, rotation=<rot -90 -0.0148 0.0149>}]
 neck_01 [{shapeType=Capsule, length=1.8929, radius=5.5945, center=<2.5531 0.3635 0>, rotation=<rot -90 -0.0049 0.0049>}]
 pelvis [{shapeType=Capsule, length=6.7079, radius=12.3292, center=<-3.3568 -0.6243 0>, rotation=<rot 0 90 0>}]
 spine_04 [{shapeType=Capsule, length=4.932, radius=11.873, center=<3.0414 -0.0515 -0.0025>, rotation=<rot -0.0004 90.4483 0.0013>}]
 spine_05 [{shapeType=Capsule, length=9.301, radius=11.4402, center=<3.4065 -1.2089 0.0001>}]
 thigh_l [{shapeType=Capsule, length=32.1783, radius=7.9305, center=<-19.5847 0 -0.8714>, rotation=<rot 90 0.0079 0.0094>}]
 thigh_r [{shapeType=Capsule, length=32.1783, radius=7.9305, center=<19.5847 0 0.8714>, rotation=<rot 90 0.0079 0.0094>}]
 upperarm_l [{shapeType=Capsule, length=24.3038, radius=6.3, center=<10.1374 0 -0.2114>, rotation=<rot -88.7876 -179.9763 179.9773>}]
 upperarm_r [{shapeType=Capsule, length=24.3038, radius=6.3, center=<-10.1374 0 0.2114>, rotation=<rot -88.7876 -179.9763 179.9773>}]
}
constraints []
